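# Unlit translucent Niagara material for the Fearless Flight Menace pickaxe trail, drawing a cel-shaded cartoon streak.
# A UV-built trail shape, modulated by panning motion lines, minus two dynamically offset Voronoi noises is thresholded into outer, middle and inner color bands.
# Screen-space, distance-scaled Ben Day dots shade the emissive color.
# Particle alpha drives a quantized texture dissolve, with optional lead mask, depth fade on non-mobile and camera offset.
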Material /BRCosmetics/Weapons/FORT_Melee/Pickaxe_FearlessFlightMenace/FX/M_FearlessFlight_Menace_Trail
graph {
 properties {
  bUsedWithNiagaraMeshParticles=true
  bUsedWithNiagaraRibbons=true
  bUsedWithNiagaraSprites=true
  blendMode=BLEND_Translucent
  pixelDepthOffsetMode=PDOM_Legacy
  shadingModel=MSM_Unlit
  usedShadingModels=Unlit
 }
 nodes [
  n1 = TextureCoordinate()
  n2 = ComponentMask(None=@n1) {g=true}
  n3 = Sine(None=Multiply(A=@n2) {constB=0.5})
  n4 = LinearInterpolate(
   A=LinearInterpolate(A=@n3, Alpha=@n2) {constB=0}
   B=@n3
   Alpha=ScalarParameter() {group=Trail Alpha Values, parameterName=Trail Center, sliderMax=1}
  )
  n5 = OneMinus(None=ComponentMask(None=@n1) {r=true})
  n6 = LinearInterpolate(
   A=@n5
   B=Power(Base=@n5, Exp=ScalarParameter() {defaultValue=3, group=Trail Alpha Values, parameterName=Trail_Shape_Intensity}) {constExponent=3}
   Alpha=@n2
  )
  n7 = Multiply(
   A=Saturate(None=Multiply(A=@n4, B=@n6))
   B=Add(
    A=ScalarParameter() {defaultValue=3, group=Trail Alpha Values, parameterName=Trail_Shape_Size}
    B=DynamicParameter() {defaultValue=<rgba 0 0 1 1>, paramNames [Offset U, Trail Shape Size, Param3, Param4]}.Trail Shape Size
   )
  ) {constB=3}
  n8 = MaterialFunctionCall(Float2=TextureCoordinate()) {materialFunction=&BreakOutFloat2Components}
  n9 = Multiply(A=@n8.G)
  n10 = TextureSampleParameter2D(UVs=Panner(Coordinate=AppendVector(A=@n9, B=@n8.R)) {speedX=0.3}) {
   group=Motion Lines
   parameterName=Motion Lines Texture
   samplerSource=SSM_Wrap_WorldGroupSettings
   texture=/Game/Effects/Fort_Effects/Textures/Outlander/T_InTheZone_LightRays_01
  }
  n11 = Multiply(A=@n8.G) {constB=2}
  n12 = TextureSampleParameter2D(UVs=Panner(Coordinate=AppendVector(A=@n11, B=@n8.R)) {speedX=-0.7}) {
   group=Motion Lines
   parameterName=Motion Lines Texture
   samplerSource=SSM_Wrap_WorldGroupSettings
   texture=&T_InTheZone_LightRays_01
  }
  n13 = Multiply(A=Add(A=Power(Base=@n10.R) {constExponent=1}, B=Power(Base=@n12.R))) {constB=0.2}
  n14 = Add(
   A=DynamicParameter() {defaultValue=<rgba 0 0 1 1>, paramNames [Offset U, Trail Shape Size, Param3, Param4]}.Offset U
   B=ScalarParameter() {defaultValue=0.3, group=Voronoi Noise 1, parameterName=Voronoi_1_Offset_U, sortPriority=50}
  )
  n15 = MaterialFunctionCall(Float2=TextureCoordinate()) {materialFunction=&BreakOutFloat2Components}
  n16 = Multiply(A=ScalarParameter() {defaultValue=0.2, group=Voronoi Noise 1, parameterName=Voronoi_1_Scale_U}, B=@n15.R) {constB=0.5}
  n17 = ScalarParameter() {defaultValue=1, group=Voronoi Noise 1, parameterName=Voronoi_1_Scale_V}
  n18 = TextureSampleParameter2D(UVs=AppendVector(A=Add(A=@n14, B=@n16) {constB=0.3}, B=Multiply(A=@n15.G, B=@n17) {constB=3})) {
   group=Voronoi Noise 1
   parameterName=TexVoronoi1
   samplerType=SAMPLERTYPE_LinearGrayscale
   texture=/BRCosmetics/Weapons/FORT_Melee/Pickaxe_FearlessFlightMenace/FX/Textures/T_FearlessFlight_Menace_Voronoi_1
  }
  n19 = Multiply(
   A=Power(Base=@n18.RGB, Exp=ScalarParameter() {defaultValue=2, group=Voronoi Noise 1, parameterName=Voronoi_1_Power})
   B=ScalarParameter() {defaultValue=2, group=Voronoi Noise 1, parameterName=Voronoi_1_Hardness}
  ) {constB=2}
  n20 = Add(
   A=DynamicParameter() {defaultValue=<rgba 0 0 1 1>, paramNames [Offset U, Trail Shape Size, Param3, Param4]}.Offset U
   B=ScalarParameter() {defaultValue=0.3, group=Voronoi Noise 2, parameterName=Voronoi_2_Offset_U, sortPriority=50}
  )
  n21 = Multiply(A=@n15.R, B=ScalarParameter() {defaultValue=0.2, group=Voronoi Noise 2, parameterName=Voronoi_2_Scale_U}) {constB=0.5}
  n22 = ScalarParameter() {defaultValue=1, group=Voronoi Noise 2, parameterName=Voronoi_2_Scale_V}
  n23 = TextureSampleParameter2D(UVs=AppendVector(A=Add(A=@n20, B=@n21) {constB=0.3}, B=Multiply(A=@n15.G, B=@n22))) {
   group=Voronoi Noise 2
   parameterName=TexVoronoi2
   samplerType=SAMPLERTYPE_LinearGrayscale
   texture=/BRCosmetics/Weapons/FORT_Melee/Pickaxe_FearlessFlightMenace/FX/Textures/T_FearlessFlight_Menace_Voronoi_2
  }
  n24 = Multiply(
   A=Power(Base=@n23.RGB, Exp=ScalarParameter() {defaultValue=0.8, group=Voronoi Noise 2, parameterName=Voronoi_2_Power}) {constExponent=0.8}
   B=ScalarParameter() {defaultValue=1, group=Voronoi Noise 2, parameterName=Voronoi_2_Hardness}
  )
  n25 = Subtract(
   A=Multiply(A=Saturate(None=@n7), B=LinearInterpolate(Alpha=@n13) {constA=0.85})
   B=Multiply(A=Saturate(None=@n19), B=Saturate(None=@n24)) {constA=1}
  )
  n26 = Step(Y=OneMinus(None=ScalarParameter() {defaultValue=1, group=Trail Color, parameterName=Outer_Color_Coverage, sliderMax=1}), X=@n25) {constX=0.9}
  n27 = Step(Y=OneMinus(None=ScalarParameter() {defaultValue=0.48, group=Trail Color, parameterName=Inner_Color_Coverage, sliderMax=1}), X=@n25) {constY=0.52}
  n28 = Step(Y=OneMinus(None=ScalarParameter() {defaultValue=0.7, group=Trail Color, parameterName=Middle_Color_Coverage, sliderMax=1}), X=@n25) {constY=0.3}
  n29 = LinearInterpolate(A=LinearInterpolate(A=@n26, B=@n27), B=@n28)
  LinearInterpolate(A=@n26, B=@n29, Alpha=ScalarParameter() {group=Trail Alpha Values, parameterName=Use_Stepped_Alpha, sliderMax=1, sortPriority=40}) {constAlpha=0}
  n30 = LinearInterpolate(B=VectorParameter() {group=Trail Color, parameterName=Outer_Color, defaultValue=<rgba 1 0.03 0.0462 1>}.RGB, Alpha=@n26)
  n31 = VectorParameter() {group=Trail Color, parameterName=Middle_Color, defaultValue=<rgba 0.23 0.0363 0.0322 1>}
  n32 = VectorParameter() {group=Trail Color, parameterName=Inner_Color, defaultValue=<rgba 2 0.289 0.26 1>}
  n33 = TextureSampleParameter2D() {
   group=Trail Color
   parameterName=InnerColorTexture
   sortPriority=36
   texture=/BRCosmetics/Weapons/FORT_Melee/Pickaxe_Rockstar_Female/FX/T_RockStar_Rainbow_color
  }
  n34 = Saturate(None=ScalarParameter() {group=Trail Color, parameterName=Use Texture for Inner Color, sliderMax=1, sortPriority=35})
  n35 = Multiply(
   A=ParticleColor().RGB
   B=LinearInterpolate(A=LinearInterpolate(A=@n30, B=@n31.RGB, Alpha=@n28), B=LinearInterpolate(A=@n32.RGB, B=@n33.RGB, Alpha=@n34), Alpha=@n27)
  )
  n36 = ScalarParameter() {defaultValue=350, group="Ben Day Dots [Old School Cartoon]", parameterName=DotsAmount, sliderMax=512, sortPriority=20}
  n37 = Subtract(A=Distance(A=ActorPositionWS(), B=CameraPositionWS())) {constB=100}
  n38 = Multiply(A=@n36, B=Saturate(None=Divide(A=@n37) {constB=400})) {constB=0.5}
  n39 = Rotator(Coordinate=Multiply(A=ScreenPosition().ViewportUV, B=LinearInterpolate(A=@n36, B=@n38)), Time=Constant() {r=1}) {speed=0.75}
  n40 = Distance(A=Add(A=Frac(None=@n39)) {constB=-0.5}, B=Constant())
  n41 = LinearInterpolate(Alpha=Saturate(None=Add(A=@n40) {constB=-0.1, desc=to get full transition})) {constA=0.22}
  n42 = ScalarParameter() {defaultValue=1.4, group="Ben Day Dots [Old School Cartoon]", parameterName=Ben_Day_Dots_Intensity}
  n43 = OneMinus(None=@n29)
  MaterialFunctionCall(Emissive=LinearInterpolate(A=@n35, B=Multiply(A=@n35, B=@n41), Alpha=Multiply(A=@n43, B=@n42))) {materialFunction=&MF_FX_EmissiveAdjustments}
  MaterialFunctionCall(Offset Amount=ScalarParameter() {parameterName=TrailCameraOffset}) {materialFunction=/Engine/Functions/Engine_MaterialFunctions02/WorldPositionOffset/CameraOffset}
  ScalarParameter() {parameterName=FadeGradientType}
  n44 = ParticleColor()
  n45 = ScalarParameter() {group=Dissolve, parameterName=Preview Dissolve - THIS SHOULD BE 0 NORMALLY, sliderMax=1}
  n46 = Floor(None=ScalarParameter() {defaultValue=4, group=Dissolve, parameterName=Dissolve Steps})
  n47 = Multiply(A=LinearInterpolate(A=OneMinus(None=@n44.A), B=@n45, Alpha=@n45), B=@n46)
  n48 = Saturate(None=Divide(A=Floor(None=@n47), B=@n46))
  n49 = TextureCoordinate()
  n50 = Multiply(A=ComponentMask(None=@n49) {r=true}, B=ScalarParameter() {defaultValue=0.4, group=Dissolve, parameterName=Dissolve_Scale_U}) {constB=0.4}
  n51 = ComponentMask(None=@n49) {g=true}
  n52 = ScalarParameter() {defaultValue=0.45, group=Dissolve, parameterName=Dissolve_Scale_V}
  n53 = TextureSampleParameter2D(UVs=AppendVector(A=Add(A=@n50) {constB=0.3}, B=Multiply(A=@n51, B=@n52) {constB=0.3})) {
   group=Dissolve
   parameterName=Tex_Dissolve
   samplerType=SAMPLERTYPE_LinearGrayscale
   texture=/BRCosmetics/Weapons/FORT_Melee/Pickaxe_FearlessFlightMenace/FX/Textures/T_FearlessFlight_Menace_Trail_Dissolve_Frac
  }
  n54 = ComponentMask(None=TextureCoordinate()) {r=true}
  n55 = Power(Base=Sine(None=Multiply(A=@n54) {constB=0.5})) {constExponent=1.3}
  n56 = OneMinus(None=ComponentMask(None=TextureCoordinate() {coordinateIndex=1}) {r=true})
  n57 = LinearInterpolate(A=@n56, Alpha=SmoothStep(Value=@n56) {constMin=0.9}) {constB=0}
  n58 = Multiply(A=@n53.RGB, B=Saturate(None=LinearInterpolate(A=@n55, B=@n56, Alpha=@n57)))
  n59 = Saturate(None=Multiply(A=ScalarParameter() {defaultValue=1.5, group=Trail Alpha Values, parameterName=Trail_Alpha_Intensity}, B=@n26) {constB=1.2})
  n60 = LinearInterpolate(A=Multiply(A=@n58) {constB=0.1}, B=@n58, Alpha=@n59)
  n61 = TextureCoordinate()
  n62 = Multiply(
   A=Add(A=ComponentMask(None=@n61) {r=true}, B=ScalarParameter() {defaultValue=1, parameterName=LeadMaskOffset})
   B=ScalarParameter() {defaultValue=0.5, parameterName=EdgeIntensity}
  )
  n63 = Multiply(
   A=Multiply(A=Step(Y=@n48, X=@n60), B=@n59)
   B=LinearInterpolate(B=Saturate(None=@n62), Alpha=ScalarParameter() {parameterName=UseLeadMask}) {constA=1}
  )
  ShadingPathSwitch(Default=DepthFade(Opacity=@n63, FadeDistance=ScalarParameter() {group=Alpha, parameterName=FadeDistance, sortPriority=2}), Mobile=@n63) {desc=Migrated from FeatureLevelSwitch}
 ]
}
