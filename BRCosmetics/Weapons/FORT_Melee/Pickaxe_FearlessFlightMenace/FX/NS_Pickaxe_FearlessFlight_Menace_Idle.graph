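# Niagara idle effect for the Fearless Flight Menace pickaxe, producing a red glow from a single Glow emitter.
# A burst spawns particles that do not die when their lifetime elapses, drawn by two mesh renderers and a light renderer.
# Each update selects primary color and a secondary glow color by time of day, both dimmest at night.
# The mesh orientation module is disabled, and none of the shown modules read the declared user parameters.
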NiagaraSystem /BRCosmetics/Weapons/FORT_Melee/Pickaxe_FearlessFlightMenace/FX/NS_Pickaxe_FearlessFlight_Menace_Idle
dependencies {usedDataInterfaces [], usedDynamicInputs [], usedModules [&SystemState]}
emitters {Glow {
  emitterUpdateScript [
   &EmitterState(Delay First Loop Only=true, Life Cycle Mode=::NewEnumerator1, Loop Delay=0.2, Loop Duration=1, UseLoopDelay=true)
   &SpawnBurst_Instantaneous(Use Spawn Probability=false, UseLoopCountLimit=true)
  ]
  particleSpawnScript [
   &InitializeParticle(
    Mesh Scale=&Multiply_VectorByFloat(Vector=<1 1 1>, Float=1)
    Mesh Scale Mode=::NewEnumerator4
    Position Offset=<-20 8.7 0>
    UsePositionOffset=false
    Write Material Random=true
   )
   &InitialMeshOrientation(Mesh Orientation Mode=::NewEnumerator7, Rotation=<0 -0.01 0>, Use Rotation Vector=true) {enabled=false}
  ]
  particleUpdateScript [
   &ParticleState(Kill Particles When Lifetime Has Elapsed=false, Let Infinitely Lived Particles Die When Emitter Deactivates=true)
   /SetVariables(Particles.ColorSecondaryGlow:LinearColor=&FN_Select_Color_from_Time_Of_Day(Morning=<rgba 0.15 0.0045 0.0239 1>, Day=<rgba 0.15 0.0045 0.0239 1>, Evening=<rgba 0.15 0.0045 0.0239 1>, Night=<rgba 0.1 0.004 0.0168 1>)) {isSetParameters=true}
   &Color(Color=&FN_Select_Color_from_Time_Of_Day(Morning=<rgba 3 0.0448 0.03 1>, Day=<rgba 4 0.0598 0.04 1>, Evening=<rgba 3 0.0448 0.03 1>, Night=<rgba 2 0.0299 0.02 1>))
  ]
  rendererClasses [&NiagaraLightRendererProperties, &NiagaraMeshRendererProperties]
  renderers [
   {rendererClass=&NiagaraLightRendererProperties, rendererIndex=0}
   {rendererClass=&NiagaraMeshRendererProperties, rendererIndex=1}
   {rendererClass=&NiagaraMeshRendererProperties, rendererIndex=2}
  ]
 }}
summary {
 systemName=NS_Pickaxe_FearlessFlight_Menace_Idle
 emitters [{emitterName=Glow, rendererClasses [&NiagaraLightRendererProperties, &NiagaraMeshRendererProperties]}]
 userVariables [
  User.bisSwinging {type=NiagaraBool}
  User.Color A {type=LinearColor, description="", defaultValue {struct=LinearColor, value=<rgba 1 1 1 1>}}
  User.NewVector3f {type=Vector3f, description="", defaultValue {struct=Vector3f, value=<0 0 0>}}
  User.RibbonIDOffset {type=NiagaraInt32}
  User.SkeletalMesh {
   type=NiagaraDataInterfaceSkeletalMesh
   defaultValue {
    struct=NiagaraExt_VariableValue_DataInterface
    value {dataInterface=&self:NiagaraDataInterfaceSkeletalMesh_0}
   }
  }
  User.Variant {type=NiagaraFloat, defaultValue=0, description=""}
 ]
}
