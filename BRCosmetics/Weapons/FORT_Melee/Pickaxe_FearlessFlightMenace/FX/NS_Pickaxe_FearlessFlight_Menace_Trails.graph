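# Niagara swing-trail system for the Fearless Flight Menace pickaxe, built from three ribbon emitters that follow sockets or bones on User.SkeletalMesh through FN_ AnimTrail modules.
# BladeTrailBright is a brief, hot red, tapering blade streak.
# EnergyTrail is a longer-lived, double-renderer ribbon with time-of-day brightness; its outline and width-taper steps are disabled.
# MotionBlur_Short adds a short red smear, and all three fade by normalized loop age.
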NiagaraSystem /BRCosmetics/Weapons/FORT_Melee/Pickaxe_FearlessFlightMenace/FX/NS_Pickaxe_FearlessFlight_Menace_Trails
dependencies {
 usedDataInterfaces [&NiagaraDataInterfaceColorCurve, &NiagaraDataInterfaceCurve]
 usedDynamicInputs []
 usedModules [&SystemState]
}
emitters {
 BladeTrailBright {
  emitterUpdateScript [
   &EmitterState(Life Cycle Mode=::NewEnumerator1, Loop Behavior=::NewEnumerator1, Loop Duration=0.74, UseLoopDelay=false)
   &FN_SpawnAnimTrail(AnimTrailWidthMode=::FromSecond, Spawn Interval=0)
  ]
  particleSpawnScript [
   &FN_InitializeAnimTrailBoneIndecies(Second Bone or Socket Index=2)
   &FN_InitializeAnimTrail(AnimTrailWidth=0.7, Color=<rgba 0.8 0.4916 0.16 1>, Lifetime=0.05) {enabled=false}
   &FN_InitializeAnimTrail(AnimTrailWidth=3, Color=<rgba 8 0.1331 0.12 0.45>, Lifetime=0.1) {module=FN_InitializeAnimTrail001}
  ]
  particleUpdateScript [
   &FN_UpdateAnimTrail(Locked Particle RibbonWidth=1, SkeletalMesh=$User.SkeletalMesh) {module=FN_UpdateAnimTrail001}
   &ScaleColor(
    Scale Alpha=&FloatFromCurve(
     FloatCurve=DataInterface() {propertyValues {Curve {keys [
         {arriveTangent=-2.003, interpMode=RCIM_Cubic, leaveTangent=-2.003, tangentMode=RCTM_User, time=0.5, value=1}
         {interpMode=RCIM_Cubic, tangentMode=RCTM_User, time=1, value=0}
        ]}}}
     Scale Curve=0.8
    )
    Scale RGB=&VectorFromFloat(Value=&FloatFromCurve(FloatCurve=DataInterface() {propertyValues {Curve {keys [{arriveTangent=-13.6919, interpMode=RCIM_Cubic, leaveTangent=-13.6919, tangentMode=RCTM_User, time=0.05, value=4}, <at 0.4 1 RCIM_Cubic>]}}}))
    ScaleA=true
    ScaleRGB=true
   )
   &ScaleColor(
    Scale Alpha=&FloatFromCurve(
     FloatCurve=DataInterface() {propertyValues {Curve {keys [
         {arriveTangent=-0.1088, interpMode=RCIM_Cubic, leaveTangent=-0.1088, tangentMode=RCTM_User, time=0.35, value=1}
         {arriveTangent=-8.4144, interpMode=RCIM_Cubic, leaveTangent=-8.4144, tangentMode=RCTM_User, time=0.5, value=0}
        ]}}}
     CurveIndex=$Emitter.NormalizedLoopAge
    )
    ScaleA=true
    ScaleRGB=false
   ) {module=ScaleColor001}
   /SetVariables(Particles.RibbonWidth=&Multiply_Float(
     A=$Particles.Initial.RibbonWidth
     B=&FloatFromCurve(FloatCurve=DataInterface() {propertyValues {Curve {keys [
          {arriveTangent=-0.1519, interpMode=RCIM_Cubic, leaveTangent=-0.1519, tangentMode=RCTM_User, time=0.2, value=1}
          {arriveTangent=-0.2854, interpMode=RCIM_Cubic, leaveTangent=-0.2854, tangentMode=RCTM_User, time=0.6, value=0.2}
         ]}}})
    )) {isSetParameters=true}
  ]
  rendererClasses [&NiagaraRibbonRendererProperties]
  renderers [{rendererClass=&NiagaraRibbonRendererProperties}]
 }
 EnergyTrail {
  emitterUpdateScript [
   &EmitterState(Life Cycle Mode=::NewEnumerator1, Loop Behavior=::NewEnumerator1, Loop Duration=0.5, UseLoopDelay=false)
   &FN_SpawnAnimTrail(Spawn Interval=0)
  ]
  particleSpawnScript [
   &FN_InitializeAnimTrailBoneIndecies(First Bone or Socket Index=1, Second Bone or Socket Index=0)
   &FN_InitializeAnimTrail(
    AnimTrailWidth=1.3
    Color=&FN_Select_Color_from_Time_Of_Day(
     Morning=/Niagara/DynamicInputs/TypeConversions/MakeLinearColorFromFloat(Value:NiagaraFloat=1.5)
     Day=&MakeLinearColorFromFloat(Value:NiagaraFloat=1.8)
     Evening=&MakeLinearColorFromFloat(Value:NiagaraFloat=1.5)
     Night=&MakeLinearColorFromFloat(Value:NiagaraFloat=1.1)
    )
    Lifetime=0.27
   ) {module=FN_InitializeAnimTrail001}
   &AddVelocity(Velocity=<1000 0 0>) {enabled=false}
  ]
  particleUpdateScript [
   &FN_UpdateAnimTrail(Locked Particle RibbonWidth=1, SkeletalMesh=$User.SkeletalMesh) {module=FN_UpdateAnimTrail001}
   &ScaleColor(
    Scale Alpha=&FloatFromCurve(
     FloatCurve=DataInterface() {propertyValues {Curve {keys [
         {arriveTangent=-5.5968, time=0.15, value=1}
         {arriveTangent=-5, time=0.16, value=0.68}
         {arriveTangent=-3, leaveTangent=-3, time=0.21, value=0.68}
         <at 0.22 0.35>
         <at 0.28 0.35>
         {leaveTangent=-1.4826, time=0.29, value=0}
        ]}}}
     CurveIndex=&Multiply_Float(A=$Emitter.LoopedAge, B=0.8)
    )
    Scale RGB=&MakeVectorFromLinearColorRGB(Color=&ColorFromCurve(ColorCurve=DataInterface() {propertyValues {
        AlphaCurve {keys []}
        BlueCurve {keys [<at 0 0.07 RCIM_Cubic>,<at 0.035 0.0265 RCIM_Cubic>]}
        GreenCurve {keys [<at 0 0.101 RCIM_Cubic>,<at 0.035 0.01 RCIM_Cubic>]}
        RedCurve {keys [<at 0 1 RCIM_Cubic>,<at 0.035 1 RCIM_Cubic>]}
       }}))
    ScaleA=true
    ScaleRGB=true
   ) {enabled=false}
   &ScaleColor(
    Scale Alpha=&FloatFromCurve(
     FloatCurve=DataInterface() {propertyValues {Curve {keys [
         <at 0.05 0 RCIM_Cubic>
         {arriveTangent=0.0046, interpMode=RCIM_Cubic, leaveTangent=0.0046, tangentMode=RCTM_User, time=0.1, value=1}
         {arriveTangent=-2.4915, interpMode=RCIM_Cubic, leaveTangent=-2.4915, tangentMode=RCTM_User, time=0.2, value=0.92}
         {arriveTangent=-0.9887, interpMode=RCIM_Cubic, leaveTangent=-0.9887, tangentMode=RCTM_User, time=0.45, value=0}
        ]}}}
     CurveIndex=$Emitter.NormalizedLoopAge
    )
    Scale RGB=&MakeVectorFromLinearColorRGB(Color=&ColorFromCurve(ColorCurve=DataInterface() {propertyValues {
        AlphaCurve {keys []}
        BlueCurve {keys [<at 0 0.24 RCIM_Cubic>,<at 0.0358 1>]}
        GreenCurve {keys [<at 0 0.4987 RCIM_Cubic>,<at 0.0358 1>]}
        RedCurve {keys [<at 0 8 RCIM_Cubic>,<at 0.0358 1>]}
       }}))
    ScaleA=true
    ScaleRGB=false
   ) {module=ScaleColor001}
   /SetVariables(Particles.ColorOutline:LinearColor=&MakeLinearColorFromVectorAndFloat(
     "Vector (RGB)"=&MakeVectorFromLinearColorRGB(Color=<rgba 0.3 0 0.2289 1>)
     "Float (Alpha)"=&FloatFromCurve(
      FloatCurve=DataInterface() {propertyValues {Curve {keys [
          <at 0.05 0>
          <at 0.1 0.8>
          {arriveTangent=-1.1624, interpMode=RCIM_Cubic, leaveTangent=-1.1624, tangentMode=RCTM_User, time=0.24, value=0.8}
          {arriveTangent=-0.7585, interpMode=RCIM_Cubic, leaveTangent=-0.7585, tangentMode=RCTM_User, time=0.5, value=0}
         ]}}}
      CurveIndex=$Emitter.NormalizedLoopAge
     )
    )) {enabled=false, isSetParameters=true}
   &SolveForcesAndVelocity(Clamp Velocity=false, Limit Acceleration=false)
   &DynamicMaterialParameters(
    Index 0 Param 1=$Emitter.NormalizedLoopAge
    Module.Param0WriteEnabled=true
    Module.Param1WriteEnabled=true
    Module.Param2WriteEnabled=true
    Module.Param3WriteEnabled=true
   ) {enabled=false}
   /SetVariables(Particles.RibbonWidth=&Multiply_Float(
     A=$Particles.Initial.RibbonWidth
     B=&FloatFromCurve(FloatCurve=DataInterface() {propertyValues {Curve {keys [
          {arriveTangent=-0.1519, interpMode=RCIM_Cubic, leaveTangent=-0.1519, tangentMode=RCTM_User, time=0.1, value=1}
          {arriveTangent=-2.2793, interpMode=RCIM_Cubic, leaveTangent=-2.2793, tangentMode=RCTM_User, time=0.7, value=0.5}
         ]}}})
    )) {enabled=false, isSetParameters=true}
   /SetVariables(Particles.RibbonWidthOutline:NiagaraFloat=&Multiply_Float(A=$Particles.RibbonWidth, B=1.2)) {enabled=false, isSetParameters=true}
  ]
  rendererClasses [&NiagaraRibbonRendererProperties]
  renderers [
   {rendererClass=&NiagaraRibbonRendererProperties}
   {rendererClass=&NiagaraRibbonRendererProperties, rendererIndex=1}
  ]
 }
 MotionBlur_Short {
  emitterUpdateScript [
   &EmitterState(Life Cycle Mode=::NewEnumerator1, Loop Behavior=::NewEnumerator1, Loop Duration=0.3, UseLoopDelay=false)
   &FN_SpawnAnimTrail(Spawn Interval=0)
  ]
  particleSpawnScript [
   &FN_InitializeAnimTrailBoneIndecies(First Bone or Socket Index=1, Second Bone or Socket Index=0)
   &FN_InitializeAnimTrail(
    AnimTrailWidth=1.25
    Color=&MakeLinearColorFromVectorAndFloat(
     "Vector (RGB)"=&MakeVectorFromLinearColorRGB(Color=&FN_Select_Color_from_Time_Of_Day(Morning=<rgba 1 0 0 1>, Day=<rgba 1 0 0 1>, Evening=<rgba 1 0 0 1>, Night=<rgba 1 0 0 1>))
     "Float (Alpha)"=1
    )
    Lifetime=0.07
   ) {module=FN_InitializeAnimTrail001}
   &AddVelocity(Velocity=<1000 0 0>) {enabled=false}
  ]
  particleUpdateScript [
   &FN_UpdateAnimTrail(Locked Particle RibbonWidth=1, SkeletalMesh=$User.SkeletalMesh) {module=FN_UpdateAnimTrail001}
   &ScaleColor(
    Scale Alpha=&FloatFromCurve(
     FloatCurve=DataInterface() {propertyValues {Curve {keys [
         {arriveTangent=-1.0572, interpMode=RCIM_Cubic, leaveTangent=-1.0572, tangentMode=RCTM_User, time=0.45, value=1}
         {arriveTangent=-1.7063, interpMode=RCIM_Cubic, leaveTangent=-1.7063, tangentMode=RCTM_User, time=0.67, value=0}
        ]}}}
     CurveIndex=$Emitter.NormalizedLoopAge
    )
    Scale RGB=&MakeVectorFromLinearColorRGB(Color=&ColorFromCurve(ColorCurve=DataInterface() {propertyValues {
        AlphaCurve {keys []}
        BlueCurve {keys [<at 0.1786 0>,<at 0.9697 0>]}
        GreenCurve {keys [<at 0.1786 0>,<at 0.9697 0>]}
        RedCurve {keys [<at 0.1786 1.1>,<at 0.9697 0.4>]}
       }}))
    ScaleA=true
    ScaleRGB=false
   ) {module=ScaleColor001}
   &SolveForcesAndVelocity(Clamp Velocity=false, Limit Acceleration=false)
  ]
  rendererClasses [&NiagaraRibbonRendererProperties]
  renderers [{rendererClass=&NiagaraRibbonRendererProperties}]
 }
}
summary {
 systemName=NS_Pickaxe_FearlessFlight_Menace_Trails
 emitters [
  {emitterName=EnergyTrail, rendererClasses [&NiagaraRibbonRendererProperties]}
  {emitterName=BladeTrailBright, rendererClasses [&NiagaraRibbonRendererProperties]}
  {emitterName=MotionBlur_Short, rendererClasses [&NiagaraRibbonRendererProperties]}
 ]
 userVariables [
  User.bisSwinging {type=NiagaraBool}
  User.Color A {type=LinearColor, description="", defaultValue {struct=LinearColor, value=<rgba 1 1 1 1>}}
  User.ColorSwitcher {type=NiagaraFloat, defaultValue=0, description=""}
  User.NewVector3f {type=Vector3f, description="", defaultValue {struct=Vector3f, value=<0 0 0>}}
  User.RibbonIDOffset {type=NiagaraInt32}
  User.SkeletalMesh {
   type=NiagaraDataInterfaceSkeletalMesh
   defaultValue {
    struct=NiagaraExt_VariableValue_DataInterface
    value {dataInterface=&self:NiagaraDataInterfaceSkeletalMesh_0}
   }
  }
  User.Variant {type=NiagaraFloat, defaultValue=0, description=""}
 ]
}
